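# Livery skin function for the Thrice vehicle body that overrides BaseColor and EmissiveColor on incoming material attributes.
# Parameters supply team, accent and paint colors, alongside detail textures blended by the skin mask, plus AO.
# Diffuse blends a darkened color toward another through a time-pulsed Fresnel and UV1 gradient.
# Emissive layers colors through three offset, panning triangle-wave bands computed as one float3 for cost, with comments noting parity breaks.
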MaterialFunction /VehicleCosmetics/Bodies/Thrice/Skins/MF_Skin_ThriceLivery1
graph {nodes [
  n1 = NamedRerouteUsage()
  n2 = NamedRerouteUsage()
  n3 = NamedRerouteUsage()
  Composite!(DetailBaseTexture=@n1, DetailAccentTexture=@n2, SkinMaskTexture=@n3) {subgraphName=Details}
  Composite!() {subgraphName=Vectors}
  n4 = NamedRerouteUsage()
  Composite!(UV Detail=@n4) {subgraphName=BodyTextures}
  Fresnel() {baseReflectFraction=-0.2, exponent=2}
  n5 = FunctionInput() {inputType=FunctionInput_MaterialAttributes}
  FunctionOutput(None=SetMaterialAttributes!(MaterialAttributes=GetMaterialAttributes(None=@n5).MaterialAttributes, BaseColor=NamedRerouteUsage(), EmissiveColor=NamedRerouteUsage()) {attributeSetTypes [#1, #2]})
  MaterialExpressionReroute_0 = Reroute(None=@n1)
  MaterialExpressionReroute_1 = Reroute(None=@n2)
  MaterialExpressionReroute_2 = Reroute(None=@n3)
  n6 = LinearInterpolate(
   A=@MaterialExpressionReroute_0.DetailBaseTexture
   B=@MaterialExpressionReroute_1.DetailAccentTexture
   Alpha=Saturate(None=MaterialFunctionCall(Color=@MaterialExpressionReroute_2.SkinMaskTexture) {materialFunction=&SplitComponents}.G)
  ) {constA=1}
  MaterialExpressionReroute_3 = Reroute(None=@n6)
  NamedRerouteDeclaration(None=@MaterialExpressionReroute_3.DetailsMasked) {name=DetailsMasked, nodeColor=<rgba 0 1 0.7656 1>}
  NamedRerouteDeclaration(None=VectorParameter() {group=NonStandardParameters, parameterName=Alt_PaintColor, defaultValue=<rgba 0.9828 0.4545 0 1>}.RGB) {name=Alt_PaintColor, nodeColor=<rgba 1 0.2344 0 1>}
  n7 = VectorParameter() {group=Color, parameterName=BaseColor, sortPriority=1, defaultValue=<rgba 0 0.2 1 1>}
  MaterialExpressionReroute_23 = Reroute(None=@n7.RGB)
  NamedRerouteDeclaration(None=@MaterialExpressionReroute_23.TeamColor) {name=TeamColor, nodeColor=<rgba 0.9219 0 1 1>}
  n8 = VectorParameter() {group=Color, parameterName=AccentColor, sortPriority=2, defaultValue=<rgba 1 1 1 1>}
  MaterialExpressionReroute_24 = Reroute(None=@n8.RGB)
  NamedRerouteDeclaration(None=@MaterialExpressionReroute_24.AccentColor) {name=AccentColor, nodeColor=<rgba 0.9219 0 1 1>}
  n9 = VectorParameter() {group=Color, parameterName=PaintColor, sortPriority=3, defaultValue=<rgba 0 1 0 1>}
  MaterialExpressionReroute_25 = Reroute(None=@n9.RGB)
  NamedRerouteDeclaration(None=@MaterialExpressionReroute_25.PaintColor) {name=PaintColor, nodeColor=<rgba 0.9219 0 1 1>}
  n10 = NamedRerouteUsage()
  n11 = NamedRerouteUsage()
  NamedRerouteDeclaration(None=Multiply(A=Multiply(A=@n10, B=@n11), B=NamedRerouteUsage())) {name=BaseColor, nodeColor=<rgba 0 1 0.7656 1>}
  n12 = Multiply(A=NamedRerouteUsage()) {constB=0.015, desc="Breaking parity: ENV Brightness on Prime adds in more team color; without this, the decal appears too black"}
  n13 = NamedRerouteUsage()
  n14 = ComponentMask(None=TextureCoordinate() {coordinateIndex=1}) {g=true}
  n15 = Fresnel() {baseReflectFraction=0, exponent=1}
  n16 = OneMinus(None=@n15)
  n17 = Saturate(None=Add(A=ConstantBiasScale(None=@n14) {bias=-0.5, scale=0.1}, B=ConstantBiasScale(None=@n16) {bias=-0.7}))
  n18 = Saturate(None=ConstantBiasScale(None=@n15) {bias=-0.6, scale=0.2})
  n19 = Time()
  n20 = ConstantBiasScale(None=Abs(None=Sine(None=@n19) {period=30})) {bias=-0.5, scale=2}
  n21 = Saturate(None=Multiply(A=Add(A=@n17, B=@n18), B=Clamp(None=@n20) {minDefault=0.2}))
  NamedRerouteDeclaration(None=Saturate(None=LinearInterpolate(A=@n12, B=@n13, Alpha=@n21) {desc=OutDiffuse})) {name=Livery1Diffuse_Out, nodeColor=<rgba 0.75 0.25 0.5 1>}
  n22 = NamedRerouteUsage()
  n23 = ConstantBiasScale(None=NamedRerouteUsage()) {bias=0, scale=0.25}
  n24 = NamedRerouteUsage()
  n25 = Multiply(A=Time()) {constB=0.1}
  n26 = TextureCoordinate() {uTiling=0.75, vTiling=0.75}
  n27 = ScalarParameter() {defaultValue=-1, group=NonStandardParameters, parameterName=PanningOffset}
  n28 = Add(A=@n25, B=@n27)
  n29 = Add(A=@n28, B=@n27)
  n30 = MaterialFunctionCall(
   X=ComponentMask(None=Add(A=@n25, B=@n26)) {g=true}
   Y=ComponentMask(None=Add(A=@n26, B=@n28)) {g=true}
   Z=ComponentMask(None=Add(A=@n26, B=@n29)) {g=true}
  ) {
   desc=Break in parity: Combining these into a vector3 and doing the same math on it is cheaper than doing them individually
   materialFunction=&MakeFloat3
  }
  n31 = Floor(None=@n30.Result)
  n32 = Abs(None=ConstantBiasScale(None=Subtract(A=@n30.Result, B=@n31)) {bias=-0.5, scale=2})
  n33 = MaterialFunctionCall(Float3=Clamp(None=ConstantBiasScale(None=@n32) {bias=-0.2, scale=2}) {minDefault=0.1}) {materialFunction=&BreakOutFloat3Components}
  n34 = Saturate(None=Add(A=Saturate(None=@n23), B=Multiply(A=@n24, B=@n33.R)))
  n35 = Multiply(A=NamedRerouteUsage(), B=@n33.G)
  n36 = NamedRerouteUsage()
  n37 = LinearInterpolate(
   A=LinearInterpolate(A=LinearInterpolate(B=@n22, Alpha=@n34), B=NamedRerouteUsage(), Alpha=Saturate(None=@n35))
   B=NamedRerouteUsage()
   Alpha=Saturate(None=Multiply(A=@n33.B, B=@n36))
  )
  n38 = ScalarParameter() {defaultValue=5, group=NonStandardParameters, parameterName=EmissiveIntensity}
  NamedRerouteDeclaration(None=Saturate(None=Multiply(A=@n37, B=@n38) {desc=OutEmissive})) {name=Livery1Emissive_Out, nodeColor=<rgba 0.75 0.25 0.5 1>}
  n39 = TextureSampleParameter2D() {group=Texture, parameterName=SkinMask, samplerSource=SSM_Wrap_WorldGroupSettings, texture=/VehicleCosmetics/Textures/T_Black}
  MaterialExpressionReroute_51 = Reroute(None=@n39.RGBA)
  NamedRerouteDeclaration(None=@MaterialExpressionReroute_51.SkinMask) {name=SkinMaskTexture, nodeColor=<rgba 0.6038 0.3325 0.013 1>}
  n40 = TextureSampleParameter2D() {
   group=Texture
   parameterName=BodyTextures
   samplerSource=SSM_Wrap_WorldGroupSettings
   samplerType=SAMPLERTYPE_LinearColor
   texture=/VehicleCosmetics/Textures/T_Black_Linear
  }
  MaterialExpressionReroute_50 = Reroute(None=@n40.G)
  NamedRerouteDeclaration(None=@MaterialExpressionReroute_50.AO) {name=AOTexture, nodeColor=<rgba 0.6038 0.3325 0.013 1>}
  MaterialExpressionReroute_45 = Reroute(None=@n4)
  n41 = TextureSampleParameter2D(UVs=@MaterialExpressionReroute_45.UV Detail) {group=Finish, parameterName=DetailBase, samplerSource=SSM_Wrap_WorldGroupSettings, texture=/VehicleCosmetics/Textures/T_White}
  MaterialExpressionReroute_53 = Reroute(None=@n41.RGB)
  NamedRerouteDeclaration(None=@MaterialExpressionReroute_53.DetailBase) {name=DetailBaseTexture, nodeColor=<rgba 0.6038 0.3325 0.013 1>}
  n42 = TextureSampleParameter2D(UVs=@MaterialExpressionReroute_45.UV Detail) {group=Finish, parameterName=DetailAccent, samplerSource=SSM_Wrap_WorldGroupSettings, texture=&T_White}
  MaterialExpressionReroute_54 = Reroute(None=@n42.RGB)
  NamedRerouteDeclaration(None=@MaterialExpressionReroute_54.DetailAccent) {name=DetailAccentTexture, nodeColor=<rgba 0.6038 0.3325 0.013 1>}
  n43 = MaterialFunctionCall(Float4=NamedRerouteUsage()) {materialFunction=&BreakOutFloat4Components}
  NamedRerouteDeclaration(None=@n43.R) {name=SkinMask_R, nodeColor=<rgba 0.9922 0 1 1>}
  NamedRerouteDeclaration(None=@n43.G) {name=SkinMask_G, nodeColor=<rgba 0.1484 1 0 1>}
  NamedRerouteDeclaration(None=@n43.B) {name=SkinMask_B, nodeColor=<rgba 0 0 1 1>}
  NamedRerouteDeclaration(None=@n43.A) {name=SkinMask_A, nodeColor=<rgba 1 1 1 1>}
  NamedRerouteDeclaration(None=Multiply(A=TextureCoordinate() {coordinateIndex=1}, B=ScalarParameter() {defaultValue=3, group=Finish, parameterName=DetailTiling})) {name=UV_Detail, nodeColor=<rgba 1 0.8646 0.8646 1>}
  NamedRerouteUsage()
  PinBase() {reroutePins [UV Detail {expression=&self:MaterialExpressionReroute_45}]}
  PinBase!(AO=@n40.G, SkinMask=@n39.RGBA, DetailBase=@n41.RGB, DetailAccent=@n42.RGB) {reroutePins [
    AO {expression=&self:MaterialExpressionReroute_50}
    SkinMask {expression=&self:MaterialExpressionReroute_51}
    DetailBase {expression=&self:MaterialExpressionReroute_53}
    DetailAccent {expression=&self:MaterialExpressionReroute_54}
   ]}
  PinBase!() {reroutePins [
    DetailBaseTexture {expression=&self:MaterialExpressionReroute_0}
    DetailAccentTexture {expression=&self:MaterialExpressionReroute_1}
    SkinMaskTexture {expression=&self:MaterialExpressionReroute_2}
   ]}
  PinBase!(DetailsMasked=@n6) {reroutePins [DetailsMasked {expression=&self:MaterialExpressionReroute_3}]}
  PinBase!()
  PinBase!(TeamColor=@n7.RGB, AccentColor=@n8.RGB, PaintColor=@n9.RGB) {reroutePins [
    TeamColor {expression=&self:MaterialExpressionReroute_23}
    AccentColor {expression=&self:MaterialExpressionReroute_24}
    PaintColor {expression=&self:MaterialExpressionReroute_25}
   ]}
 ]}
