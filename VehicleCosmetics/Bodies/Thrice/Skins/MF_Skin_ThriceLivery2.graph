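# Material function for the Livery2 skin of the Thrice vehicle body, overriding BaseColor and EmissiveColor on incoming material attributes.
# Livery diffuse blends Fresnel-shaded tints through masks; separate composites supply team, accent and paint colors, SkinMask channels, AO and tiled detail textures.
# Animated emissive thresholds a UV1 panning triangle wave combined with a sine pulse, then applies Fresnel falloff and an intensity parameter.
# SkinMask alpha is also sampled as a tiled, scrolling pattern.
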MaterialFunction /VehicleCosmetics/Bodies/Thrice/Skins/MF_Skin_ThriceLivery2
graph {nodes [
  n1 = NamedRerouteUsage()
  Composite!(UV Detail=@n1) {subgraphName=BodyTextures}
  n2 = NamedRerouteUsage()
  n3 = NamedRerouteUsage()
  n4 = NamedRerouteUsage()
  Composite!(DetailBaseTexture=@n2, DetailAccentTexture=@n3, SkinMaskTexture=@n4) {subgraphName=Details}
  Composite!() {subgraphName=Vectors}
  n5 = FunctionInput() {inputType=FunctionInput_MaterialAttributes}
  FunctionOutput(None=SetMaterialAttributes!(MaterialAttributes=GetMaterialAttributes(None=@n5).MaterialAttributes, BaseColor=NamedRerouteUsage(), EmissiveColor=NamedRerouteUsage()) {attributeSetTypes [#1, #2]})
  n6 = TextureSampleParameter2D() {group=Texture, parameterName=SkinMask, samplerSource=SSM_Wrap_WorldGroupSettings, texture=/VehicleCosmetics/Textures/T_Black}
  MaterialExpressionReroute_44 = Reroute(None=@n6.RGBA)
  NamedRerouteDeclaration(None=@MaterialExpressionReroute_44.SkinMask) {name=SkinMaskTexture, nodeColor=<rgba 0.6038 0.3325 0.013 1>}
  n7 = VectorParameter() {group=Color, parameterName=BaseColor, sortPriority=1, defaultValue=<rgba 0 0.2 1 1>}
  MaterialExpressionReroute_23 = Reroute(None=@n7.RGB)
  NamedRerouteDeclaration(None=@MaterialExpressionReroute_23.TeamColor) {name=TeamColor, nodeColor=<rgba 0.9219 0 1 1>}
  n8 = VectorParameter() {group=Color, parameterName=AccentColor, sortPriority=2, defaultValue=<rgba 1 1 1 1>}
  MaterialExpressionReroute_24 = Reroute(None=@n8.RGB)
  NamedRerouteDeclaration(None=@MaterialExpressionReroute_24.AccentColor) {name=AccentColor, nodeColor=<rgba 0.9219 0 1 1>}
  n9 = VectorParameter() {group=Color, parameterName=PaintColor, sortPriority=3, defaultValue=<rgba 0 1 0 1>}
  MaterialExpressionReroute_25 = Reroute(None=@n9.RGB)
  NamedRerouteDeclaration(None=@MaterialExpressionReroute_25.PaintColor) {name=PaintColor, nodeColor=<rgba 0.9219 0 1 1>}
  n10 = NamedRerouteUsage()
  n11 = NamedRerouteUsage()
  NamedRerouteDeclaration(None=Multiply(A=Multiply(A=@n10, B=@n11), B=NamedRerouteUsage())) {name=BaseColor, nodeColor=<rgba 0 1 0.7656 1>}
  n12 = TextureSampleParameter2D() {
   group=Texture
   parameterName=BodyTextures
   samplerSource=SSM_Wrap_WorldGroupSettings
   samplerType=SAMPLERTYPE_LinearColor
   texture=/VehicleCosmetics/Textures/T_Black_Linear
  }
  MaterialExpressionReroute_19 = Reroute(None=@n12.G)
  NamedRerouteDeclaration(None=@MaterialExpressionReroute_19.AO) {name=AOTexture, nodeColor=<rgba 0.6038 0.3325 0.013 1>}
  n13 = Constant3Vector() {constant=<rgba 0.015 0.015 0.015 0>}
  n14 = Constant3Vector() {constant=<rgba 0.1152 0.1283 0.1875 1>}
  n15 = Saturate(None=ConstantBiasScale(None=Fresnel() {baseReflectFraction=-0.1, exponent=4}) {bias=-0.2, scale=1})
  n16 = OneMinus(None=Fresnel() {baseReflectFraction=0, exponent=3})
  n17 = Max(A=ConstantBiasScale(None=@n16) {bias=0, scale=0.75}, B=Clamp(None=ConstantBiasScale(None=@n16) {bias=-0.95, scale=50}) {maxDefault=1.5})
  n18 = LinearInterpolate(
   A=LinearInterpolate(A=LinearInterpolate(A=@n13, B=@n14, Alpha=@n15), B=Multiply(A=@n14, B=@n17), Alpha=NamedRerouteUsage())
   B=Multiply(A=@n17, B=NamedRerouteUsage())
   Alpha=NamedRerouteUsage()
  )
  n19 = Constant3Vector()
  n20 = NamedRerouteUsage()
  NamedRerouteDeclaration(None=Saturate(None=LinearInterpolate(A=@n18, B=@n19, Alpha=@n20))) {name=Livery2Diffuse_Out, nodeColor=<rgba 0.75 0.25 0.5 1>}
  n21 = Saturate(None=ConstantBiasScale(None=@n16) {bias=-0.5, scale=1})
  n22 = NamedRerouteUsage()
  n23 = Time()
  n24 = ComponentMask(None=Panner(Coordinate=TextureCoordinate() {coordinateIndex=1}) {speedY=0.25}) {g=true}
  n25 = Floor(None=@n24)
  n26 = Abs(None=ConstantBiasScale(None=Subtract(A=@n24, B=@n25)) {bias=-0.5, scale=2})
  n27 = Max(
   A=ConstantBiasScale(None=Sine(None=@n23) {period=3}) {bias=0, scale=0.3}
   B=Saturate(None=ConstantBiasScale(None=@n26) {bias=-0.5, scale=0.6})
  )
  n28 = Saturate(None=ConstantBiasScale(None=Add(A=@n22, B=@n27)) {bias=-0.3, scale=500})
  n29 = Multiply(A=NamedRerouteUsage(), B=Multiply(A=NamedRerouteUsage(), B=Multiply(A=@n21, B=@n28)))
  NamedRerouteDeclaration(None=Multiply(A=Saturate(None=@n29), B=ScalarParameter() {defaultValue=2, group=NonStandardParameters, parameterName=EmissiveIntensity})) {name=Livery2Emissive_Out, nodeColor=<rgba 0.75 0.25 0.5 1>}
  n30 = MaterialFunctionCall(Float3=NamedRerouteUsage()) {materialFunction=&BreakOutFloat3Components}
  NamedRerouteDeclaration(None=@n30.R) {name=SkinMask_R, nodeColor=<rgba 0 1 0.5078 1>}
  NamedRerouteDeclaration(None=@n30.G) {name=SkinMask_G, nodeColor=<rgba 0.1484 1 0 1>}
  NamedRerouteDeclaration(None=@n30.B) {name=SkinMask_B, nodeColor=<rgba 0 0 1 1>}
  n31 = TextureSampleParameter2D(UVs=Panner(Coordinate=TextureCoordinate() {coordinateIndex=1, uTiling=25, vTiling=-25}) {speedY=0.025}) {group=Texture, parameterName=SkinMask, samplerSource=SSM_Wrap_WorldGroupSettings, texture=&T_Black}
  NamedRerouteDeclaration(None=@n31.A) {name=SkinMask_A, nodeColor=<rgba 1 1 1 1>}
  MaterialExpressionReroute_12 = Reroute(None=@n1)
  n32 = TextureSampleParameter2D(UVs=@MaterialExpressionReroute_12.UV Detail) {group=Finish, parameterName=DetailBase, samplerSource=SSM_Wrap_WorldGroupSettings, texture=/VehicleCosmetics/Textures/T_White}
  MaterialExpressionReroute_0 = Reroute(None=@n32.RGB)
  NamedRerouteDeclaration(None=@MaterialExpressionReroute_0.DetailBase) {name=DetailBaseTexture, nodeColor=<rgba 0.6038 0.3325 0.013 1>}
  n33 = TextureSampleParameter2D(UVs=@MaterialExpressionReroute_12.UV Detail) {group=Finish, parameterName=DetailAccent, samplerSource=SSM_Wrap_WorldGroupSettings, texture=&T_White}
  MaterialExpressionReroute_1 = Reroute(None=@n33.RGB)
  NamedRerouteDeclaration(None=@MaterialExpressionReroute_1.DetailAccent) {name=DetailAccentTexture, nodeColor=<rgba 0.6038 0.3325 0.013 1>}
  MaterialExpressionReroute_31 = Reroute(None=@n2)
  MaterialExpressionReroute_34 = Reroute(None=@n3)
  MaterialExpressionReroute_17 = Reroute(None=@n4)
  n34 = LinearInterpolate(
   A=@MaterialExpressionReroute_31.DetailBaseTexture
   B=@MaterialExpressionReroute_34.DetailAccentTexture
   Alpha=Saturate(None=MaterialFunctionCall(Color=@MaterialExpressionReroute_17.SkinMaskTexture) {materialFunction=&SplitComponents}.G)
  ) {constA=1}
  MaterialExpressionReroute_40 = Reroute(None=@n34)
  NamedRerouteDeclaration(None=@MaterialExpressionReroute_40.DetailsMasked) {name=DetailsMasked, nodeColor=<rgba 0 1 0.7656 1>}
  NamedRerouteDeclaration(None=Multiply(A=TextureCoordinate() {coordinateIndex=1}, B=ScalarParameter() {defaultValue=3, group=Finish, parameterName=DetailTiling})) {name=UV_Detail, nodeColor=<rgba 1 0.8646 0.8646 1>}
  NamedRerouteUsage()
  PinBase() {reroutePins [
    DetailBaseTexture {expression=&self:MaterialExpressionReroute_31}
    DetailAccentTexture {expression=&self:MaterialExpressionReroute_34}
    SkinMaskTexture {expression=&self:MaterialExpressionReroute_17}
   ]}
  PinBase!(DetailsMasked=@n34) {reroutePins [DetailsMasked {expression=&self:MaterialExpressionReroute_40}]}
  PinBase!() {reroutePins [UV Detail {expression=&self:MaterialExpressionReroute_12}]}
  PinBase!(AO=@n12.G, SkinMask=@n6.RGBA, DetailBase=@n32.RGB, DetailAccent=@n33.RGB) {reroutePins [
    AO {expression=&self:MaterialExpressionReroute_19}
    SkinMask {expression=&self:MaterialExpressionReroute_44}
    DetailBase {expression=&self:MaterialExpressionReroute_0}
    DetailAccent {expression=&self:MaterialExpressionReroute_1}
   ]}
  PinBase!()
  PinBase!(TeamColor=@n7.RGB, AccentColor=@n8.RGB, PaintColor=@n9.RGB) {reroutePins [
    TeamColor {expression=&self:MaterialExpressionReroute_23}
    AccentColor {expression=&self:MaterialExpressionReroute_24}
    PaintColor {expression=&self:MaterialExpressionReroute_25}
   ]}
 ]}
