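# Clear-coat vehicle body material for the Thrice livery skin, usable on skeletal meshes and with mesh deformers.
# The shared vehicle body base function feeds one of two livery functions, chosen by the UseLivery2 static switch.
# The result passes through vehicle damage and de-rez transition functions.
# Clear-coat bottom normal is extracted after the livery stage and routed by named reroute to the clear-coat normal custom output.
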Material /VehicleCosmetics/Bodies/Thrice/Skins/MAT_Skin_ThriceLivery
graph {
 properties {
  bUseMaterialAttributes=true
  bUsedWithMeshDeformer=true
  bUsedWithSkeletalMesh=true
  pixelDepthOffsetMode=PDOM_Legacy
  shadingModel=MSM_ClearCoat
  usedShadingModels=Clear Coat
 }
 nodes [
  ClearCoatNormalCustomOutput!(None=NamedRerouteUsage())
  n1 = MaterialFunctionCall() {materialFunction=/VehicleCosmetics/Materials/MaterialFunctions/MF_Vehicle_Body_Base}
  n2 = GetMaterialAttributes(None=StaticSwitchParameter(
    True=MaterialFunctionCall!(In=@n1.Result) {materialFunction=&/MF_Skin_ThriceLivery2}.Result
    False=MaterialFunctionCall!(In=@n1.Result) {materialFunction=&/MF_Skin_ThriceLivery1}.Result
   ) {group=NonStandardParameters, parameterName=UseLivery2}) {attributeGetTypes [#1]}
  MaterialFunctionCall(In=MaterialFunctionCall(Material=@n2.MaterialAttributes) {materialFunction=&MF_BaseVehicle_Damage}.Result) {materialFunction=&MF_VehicleDeRez_Transition}
  NamedRerouteDeclaration(None=@n2.ClearCoatBottomNormal) {name=BodyBase_BottomNormal, nodeColor=<rgba 0 1 0.3672 1>}
 ]
}
